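# AssetAction_FixOfferTextures is an editor utility blueprint, kept among never-cook editor tools, that exposes a FixOfferTextures function.
# The function hands off to Python, running one command that imports a FixOfferTextures module from a developer's OfferCatalog folder, force-reloads it and calls its Main.
# The blueprint itself holds no other logic or variables, so the texture-fixing work lives entirely in that script.
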EditorUtilityBlueprint /Game/NeverCook/EditorUtilities/IAD/AssetAction_FixOfferTextures
graphs [
 EventGraph {links [], nodes []}
 FixOfferTextures {nodes [
   FixOfferTextures() {is=title K2Node_FunctionEntry}
   > ExecutePythonCommand(PythonCommand="import Developers.ChrisMorrison.OfferCatalog.FixOfferTextures as OT; import importlib; importlib.reload(OT); OT.Main()")
  ]}
]
variables []
